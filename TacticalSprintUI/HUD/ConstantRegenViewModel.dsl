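; View model for a tactical sprint HUD progress bar that displays the unlimited energy (constant regen) state.
; Gameplay tag query waits on the pawn toggle the bar's forced visibility and enabled state.
; When enabled, it picks a material override from active gameplay cues, falling back to a default.
; A looping timer pulses the bar when the longest remaining ConstantRegen effect nears expiry, and is cleared on disable.
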(event EventInitialize
  (params (Pawn :type Pawn%20Object%20Reference))
  (|CallParentFunction Pawn)
  (Utilities|IsValid Pawn
    (:Is%20Valid
      (bind _asfort_ability_system_component (Utilities|Casting|CastToFortAbilitySystemComponent (Ability|GetAbilitySystemComponent Pawn)))
      (Variables|Components|SetAbilitySystemComponent _asfort_ability_system_component)
      (Ability|Tasks|WaitGameplayTagQueryOnActor Pawn (Variables|Default|GetUnlimitedEnergyTagQuery)
        (:then
          (Ability|Tasks|WaitGameplayTagQueryOnActor Pawn (Variables|Default|GetUnlimitedEnergyTagQuery) "WhenFalse"
            (:Triggered
              (CallFunction|ToggleUnlimitedEnergyState false))))
        (:Triggered
          (CallFunction|ToggleUnlimitedEnergyState true))))))

(event EventUninitialize
  (params (Pawn :type Pawn%20Object%20Reference))
  (Variables|Components|SetAbilitySystemComponent 0))

(event Custom|ToggleUnlimitedEnergyState
  (params (bUnlimitedEnergyState :type Boolean))
  (bind _output_get (Variables|GenericProgressBarViewModel|SetbForceVisible bUnlimitedEnergyState))
  (bind _output_get_2 (Variables|GenericProgressBarViewModel|SetbEnabled _output_get))
  (if _output_get_2
    (bind _materialoverride (CallFunction|GetDesiredMaterialOverride))
    (CallFunction|SetProgressBarOverrideMaterial _materialoverride)
    (bind _returnvalue (Utilities|Time|SetTimerbyEvent (delegate Custom|EnablePulsing) 0.25 true))
    (Variables|Default|SetUnlimitedEnergyTimerHandle _returnvalue)
    (else
      (Utilities|Time|ClearandInvalidateTimerbyHandle 0 (Variables|Default|GetUnlimitedEnergyTimerHandle)))))

(event Custom|EnablePulsing
  (bind _timeremaining (CallFunction|FindLongestTimeRemaining))
  (Variables|GenericProgressBarViewModel|SetbIsPulsing (and (<= _timeremaining 3.0) (> _timeremaining 0))))

(fn FindLongestTimeRemaining
  (bind _abilitysystemcomponent (Variables|Components|GetAbilitySystemComponent))
  (Utilities|IsValid _abilitysystemcomponent
    (:Is%20Valid
      (bind _returnvalue (GameplayEffects|GetActiveEffectswithAllTags _abilitysystemcomponent "(GameplayTags=((TagName=\"Asset.Athena.Energy.ConstantRegen\")))"))
      (for _array_element _returnvalue
        (bind _returnvalue_2 (Ability|GameplayEffect|GetActiveGameplayEffectRemainingDuration 0 _array_element))
        (if (> _returnvalue_2 (LocalVariables|GetLongestTimeFound))
          (LocalVariables|SetLongestTimeFound _returnvalue_2)))
      (return (LocalVariables|GetLongestTimeFound)))))

(fn GetDesiredMaterialOverride
  (bind _abilitysystemcomponent (Variables|Components|GetAbilitySystemComponent))
  (Utilities|IsValid _abilitysystemcomponent
    (:Is%20Valid
      (bind _keys (Utilities|Map|Keys (Variables|Default|GetAbilityMaterialMap)))
      (for _array_element _keys
        (if (GameplayCue|IsGameplayCueActive _abilitysystemcomponent _array_element)
          (return (Utilities|Map|Find (Variables|Default|GetAbilityMaterialMap) _array_element))))
      (return "/TacticalSprintUI/UI/Materials/MI_UI_Rounded_HitPointBar_EnergyUnlimited.MI_UI_Rounded_HitPointBar_EnergyUnlimited"))))

(fn ShouldBecomeVisibleImmediately
  (return 1.0 true))
