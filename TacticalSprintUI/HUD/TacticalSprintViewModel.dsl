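; View model that drives the HUD progress bar for a pawn's tactical sprint energy.
; On initialize it caches the pawn's TacticalSprintEnergyComponent, binds to its energy-changed dispatcher and enables the bar.
; Each change pushes the current energy percentage as progress while the bar is enabled.
; ShouldBecomeVisibleImmediately returns that percentage and whether it is below full; uninitialize does no unbinding.
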(event EventInitialize
  (params (Pawn :type Pawn%20Object%20Reference))
  (|CallParentFunction Pawn)
  (Utilities|IsValid Pawn
    (:Is%20Valid
      (bind _output_get (Variables|Default|SetTacticalSprintComponent (Actor|GetComponentbyClass Pawn "/TacticalSprintGame/Gameplay/TacticalSprintEnergyComponent.TacticalSprintEnergyComponent_C")))
      (Events|BindEventtoOnCurrentEnergyChanged _output_get (EventDispatchers|CreateEvent :!event OnCurrentEnergyChanged))
      (Variables|GenericProgressBarViewModel|SetbEnabled true))))

(event EventUninitialize
  (params (Pawn :type Pawn%20Object%20Reference)))

(fn OnCurrentEnergyChanged
  (params (EnergyComponent :type Fort%20Component%20Energy%20Object%20Reference) (PreviousValue :type Float%20%28single-precision%29))
  (if (Variables|GenericProgressBarViewModel|GetbEnabled)
    (Variables|GenericProgressBarViewModel|SetProgressValue (Energy|GetCurrentEnergyPercentage EnergyComponent))))

(fn ShouldBecomeVisibleImmediately
  (bind _tacticalsprintcomponent (Variables|Default|GetTacticalSprintComponent))
  (Utilities|IsValid _tacticalsprintcomponent
    (:Is%20Valid
      (return (Energy|GetCurrentEnergyPercentage _tacticalsprintcomponent) (< (Energy|GetCurrentEnergyPercentage _tacticalsprintcomponent) 1.0)))
    (:Is%20Not%20Valid
      (return 0.0 false))))
